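# UI material for Chapter 5 Season 1 Item Shop tiles, producing a gold and orange grunge background.
# A smoothstepped radial gradient blends two golds, while a linear gradient blends two darker oranges.
# The red channel of a grunge texture, sampled with the Normal sampler type, mixes the two gradients.
# All use scaled-to-fit UVs through a named reroute, with small vertical offsets on the radial gradient and grunge.
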Material /ItemShop/UI/Art/Ch5S1/M_UI_Ch5S1_ShopTiles
graph {
 properties {bUsedWithStaticMesh=false, materialDomain=MD_UI, pixelDepthOffsetMode=PDOM_Legacy, usedShadingModels=Default Lit}
 nodes [
  n1 = Constant()
  n2 = Constant() {r=0.1}
  n3 = MaterialFunctionCall(UV=Add(A=NamedRerouteUsage(), B=AppendVector(A=@n1, B=@n2)), GradientStart=Constant(), GradientEnd=Constant() {r=1}) {materialFunction=&MF_UI_RadialGradient}
  n4 = NamedRerouteUsage()
  n5 = Constant() {r=0.25}
  n6 = Constant() {r=0.5}
  n7 = Constant() {r=0.25}
  n8 = AppendVector(A=Constant(), B=Constant() {r=-0.1})
  LinearInterpolate(
   A=LinearInterpolate(
    A=Constant3Vector() {constant=<rgba 0.5972 0.3185 0.004 1>}
    B=Constant3Vector() {constant=<rgba 0.9911 0.8148 0.2423 1>}
    Alpha=SmoothStep(Value=@n3.Smooth) {constMax=1.75}
   )
   B=LinearInterpolate(
    A=Constant3Vector() {constant=<rgba 0.5029 0.1384 0.0003 1>}
    B=Constant3Vector() {constant=<rgba 0.7231 0.3968 0.0273 1>}
    Alpha=MaterialFunctionCall(UVs=@n4, GradientStart=@n5, GradientEnd=@n6, RotationAngle=@n7) {materialFunction=&MF_UI_LinearGradient}.Result
   )
   Alpha=TextureSample(UVs=Add(A=@n8, B=@n4)) {samplerType=SAMPLERTYPE_Normal, texture=/OfferCatalog/Art/Ch5S1/T_UI_Ch5S1Grunge}.R
  )
  NamedRerouteDeclaration(None=MaterialFunctionCall() {materialFunction=&MF_UI_UVScaling}.Scaled to Fit) {name=Fit, nodeColor=<rgba 0.5703 1 0 1>}
 ]
}
