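# UI material that draws an animated Chapter 5 Season 1 Item Shop background in dark blue tones, with layered clouds and light beams.
# Foreground and midground cloud masks combine linear gradients, top-bottom falloff and panning blurred noise with flow-map-distorted cloud textures.
# Light beams take the minimum of two opposite-panning, self-distorted noise samples in radial coordinates.
# A flow-mapped noise variation layer and these masks drive nested color lerps for the final blend.
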Material /ItemShop/UI/Art/Ch5S1/M_UI_Ch5S1_StoreBG
graph {
 properties {bUsedWithStaticMesh=false, materialDomain=MD_UI, pixelDepthOffsetMode=PDOM_Legacy, usedShadingModels=Default Lit}
 nodes [
  n1 = Constant3Vector() {constant=<rgba 0.0177 0.0207 0.0547 1>}
  n2 = Multiply(A=NamedRerouteUsage()) {constB=0.12}
  n3 = LinearInterpolate(A=Constant3Vector() {constant=<rgba 0.0008 0.0028 0.0104 1>}, B=Constant3Vector() {constant=<rgba 0.0774 0.1263 0.3229 1>}, Alpha=@n2)
  n4 = NamedRerouteUsage()
  n5 = NamedRerouteUsage()
  n6 = Max(A=@n4, B=@n5)
  LinearInterpolate(
   A=LinearInterpolate(A=LinearInterpolate(A=@n1, B=@n3, Alpha=@n4), B=Constant3Vector() {constant=<rgba 0.0601 0.0855 0.1875 1>}, Alpha=NamedRerouteUsage())
   B=Constant3Vector() {constant=<rgba 0.0719 0.1126 0.276 1>}
   Alpha=Min(A=OneMinus(None=@n6), B=Max(A=@n2, B=@n5))
  )
  n7 = Constant()
  n8 = Constant() {r=0.85}
  n9 = Multiply(A=NamedRerouteUsage()) {constB=3}
  n10 = MaterialFunctionCall!(UVs=Multiply(A=AppendVector(A=@n7, B=@n8), B=@n9), Subtract=Constant() {r=0.05}, Falloff=Constant() {r=1}, Intensity=Constant() {r=2}) {materialFunction=/Game/UI/Frontend/ItemShop/Art/_MaterialFunctions_ShopTheming/MF_UI_TopBottomGradient}
  n11 = MaterialFunctionCall(UVs=@n9, Texture Scale=Constant() {r=1.25}) {materialFunction=&ScaleUVsByCenter}
  n12 = Max(
   A=NamedRerouteUsage()
   B=TextureSample(UVs=Panner(Coordinate=@n11.UVs) {speedX=0.03, speedY=0.02}) {samplerType=SAMPLERTYPE_Grayscale, texture=&LowResBlurredNoise}.R
  )
  n13 = Max(A=Multiply(A=OneMinus(None=@n10.Result), B=@n12), B=@n12)
  n14 = NamedRerouteUsage()
  n15 = AppendVector(A=Constant(), B=Constant() {r=-0.2})
  n16 = Add(A=NamedRerouteUsage(), B=AppendVector(A=Constant() {r=-0.1}, B=Constant() {r=-0.1}))
  n17 = Constant() {r=-0.1}
  n18 = Constant() {r=1.25}
  n19 = Constant() {r=0.6542}
  n20 = Constant() {r=0.65}
  n21 = Max(
   A=MaterialFunctionCall(
    UVs=Add(A=@n14, B=@n15)
    GradientStart=Constant() {r=-0.1}
    GradientEnd=Constant() {r=1.25}
    RotationAngle=Constant() {r=0.75}
    GradientOffset=Constant() {r=0.6}
   ) {materialFunction=&MF_UI_LinearGradient}.Result
   B=SmoothStep(Value=MaterialFunctionCall(UVs=@n16, GradientStart=@n17, GradientEnd=@n18, RotationAngle=@n19, GradientOffset=@n20) {materialFunction=&MF_UI_LinearGradient}.Result)
  )
  n22 = Saturate(None=Multiply(A=Max(A=@n13, B=@n21), B=@n21))
  NamedRerouteDeclaration(None=SmoothStep(Value=Multiply(A=@n22) {constB=4}) {constMax=0.85}) {name=FG Clouds, nodeColor=<rgba 0.9688 0 1 1>}
  n23 = Multiply(A=NamedRerouteUsage()) {constB=1.25}
  n24 = MaterialFunctionCall!(
   UVs=Panner(Coordinate=@n23) {speedX=0.01, speedY=0.015}
   Texture Object=TextureObject() {samplerType=SAMPLERTYPE_Alpha, texture=/Game/UI/Frontend/ItemShop/Art/LastVoice/T_UI_CloudLV}
   Grayscale Flow Map=TextureSample(
    UVs=Panner(Coordinate=@n23) {speedX=-0.01, speedY=-0.01}
    Tex=TextureObject() {samplerType=SAMPLERTYPE_Grayscale, texture=/FortUILibrary/MaterialFunctions/ExampleContent/T_Cloud_R}
   ).RGB
   Flow Direction=Constant() {r=0.5}
   Flow Intensity=Constant() {r=0.09}
   Flow Speed=Constant() {r=0.25}
  ) {materialFunction=/FortUILibrary/MaterialFunctions/Effects/MF_UI_GrayscaleFlowMap_New}
  NamedRerouteDeclaration(None=Add(A=NamedRerouteUsage(), B=ComponentMask(None=@n24.Distorted Texture) {r=true})) {name=FlowMap_FG, nodeColor=<rgba 0 0 0 1>}
  n25 = ScalarParameter() {defaultValue=5, group=Radial Gradient, parameterName=Tile}
  n26 = Multiply(A=NamedRerouteUsage(), B=@n25)
  n27 = Multiply(A=@n25) {constB=0.5}
  n28 = ScalarParameter() {defaultValue=-10, parameterName=offset x}
  n29 = ScalarParameter() {defaultValue=6, parameterName=offset y}
  n30 = MaterialFunctionCall(
   Vector or UVs=Add(A=Subtract(A=@n26, B=@n27) {constB=3}, B=AppendVector(A=@n28, B=@n29)) {constB=0.5}
   Swizzle Coordinate Output=StaticBool() {value=true}
  ) {materialFunction=&VectorToRadialValue}
  n31 = Time()
  n32 = ScalarParameter() {defaultValue=0.025, parameterName=TimeFactor2}
  n33 = Multiply(A=Panner(Coordinate=@n30.Vector Converted to Angle, Time=Multiply(A=@n31, B=@n32)) {speedX=0.025, speedY=0.05}) {constB=2}
  n34 = TextureSample(UVs=Multiply(
    A=TextureSample(UVs=@n33) {samplerType=SAMPLERTYPE_Grayscale, texture=/Game/UI/Frontend/ItemShop/Art/Rumble/LowResBlurredNoise_Rumble}.R
    B=ScalarParameter() {defaultValue=1.5, parameterName=DeformIntensity}
   )) {samplerType=SAMPLERTYPE_Grayscale, texture=&LowResBlurredNoise_Rumble}
  n35 = ScalarParameter() {defaultValue=0.25}
  n36 = Saturate(None=Multiply(A=Power(Base=@n34.R, Exp=@n35) {constExponent=1}, B=ScalarParameter() {defaultValue=1.25, parameterName=Param_1}) {constB=2})
  n37 = Time()
  n38 = ScalarParameter() {defaultValue=0.05, parameterName=TimeFactor}
  n39 = Multiply(A=Panner(Coordinate=@n30.Vector Converted to Angle, Time=Multiply(A=@n37, B=@n38)) {speedX=-0.025, speedY=-0.05}) {constB=2}
  n40 = TextureSample(UVs=Multiply(
    A=TextureSample(UVs=@n39) {samplerType=SAMPLERTYPE_Grayscale, texture=&LowResBlurredNoise_Rumble}.R
    B=ScalarParameter() {defaultValue=1.5, parameterName=DeformIntensity}
   )) {samplerType=SAMPLERTYPE_Grayscale, texture=&LowResBlurredNoise_Rumble}
  n41 = ScalarParameter() {defaultValue=0.25}
  n42 = Saturate(None=Multiply(A=Power(Base=@n40.R, Exp=@n41) {constExponent=1}, B=ScalarParameter() {defaultValue=1.25, parameterName=Param_1}) {constB=2})
  NamedRerouteDeclaration(None=SmoothStep(Value=Min(A=@n36, B=@n42)) {constMax=1.25}) {name=Lightbeams, nodeColor=<rgba 1 0.8203 0 1>}
  n43 = Constant()
  n44 = Constant() {r=0.85}
  n45 = Multiply(A=NamedRerouteUsage()) {constB=3}
  n46 = MaterialFunctionCall(UVs=Multiply(A=AppendVector(A=@n43, B=@n44), B=@n45), Subtract=Constant() {r=0.05}, Falloff=Constant() {r=1}, Intensity=Constant() {r=2}) {materialFunction=&MF_UI_TopBottomGradient}
  n47 = MaterialFunctionCall(UVs=@n45, Texture Scale=Constant() {r=1.5}) {materialFunction=&ScaleUVsByCenter}
  n48 = Max(
   A=NamedRerouteUsage()
   B=TextureSample(UVs=Panner(Coordinate=@n47.UVs) {speedX=0.015}) {samplerType=SAMPLERTYPE_Grayscale, texture=&LowResBlurredNoise}.R
  )
  n49 = Max(A=Multiply(A=OneMinus(None=@n46.Result), B=@n48), B=@n48)
  n50 = Add(A=NamedRerouteUsage(), B=AppendVector(A=Constant(), B=Constant() {r=0.6}))
  n51 = Constant() {r=-0.1}
  n52 = Constant() {r=1}
  n53 = Constant() {r=0.8}
  n54 = Constant() {r=0.7}
  n55 = Add(A=NamedRerouteUsage(), B=AppendVector(A=Constant(), B=Constant() {r=0.6}))
  n56 = Constant() {r=-0.1}
  n57 = Constant() {r=0.75}
  n58 = Constant() {r=0.65}
  n59 = Constant() {r=0.65}
  n60 = SmoothStep(Value=Max(
    A=MaterialFunctionCall(UVs=@n50, GradientStart=@n51, GradientEnd=@n52, RotationAngle=@n53, GradientOffset=@n54) {materialFunction=&MF_UI_LinearGradient}.Result
    B=MaterialFunctionCall(UVs=@n55, GradientStart=@n56, GradientEnd=@n57, RotationAngle=@n58, GradientOffset=@n59) {materialFunction=&MF_UI_LinearGradient}.Result
   )) {constMax=3}
  n61 = Saturate(None=Multiply(A=Max(A=@n49, B=@n60), B=@n60))
  NamedRerouteDeclaration(None=SmoothStep(Value=Multiply(A=@n61) {constB=4}) {constMax=0.5, constMin=0.15}) {name=MG_Clouds, nodeColor=<rgba 0.9688 0 1 1>}
  n62 = MaterialFunctionCall() {materialFunction=&MF_UI_UVScaling}
  n63 = MaterialFunctionCall!(
   Texture 2D=TextureObject() {samplerType=SAMPLERTYPE_Grayscale, texture=&LowResBlurredNoise_Rumble}
   Texture UVs=Panner(Coordinate=Multiply(A=@n62.Scaled To Fill) {constB=1.2}) {speedX=-0.02, speedY=-0.013}
   FlowMap Texture 2D=TextureObject() {samplerType=SAMPLERTYPE_LinearColor, texture=/Game/UI/Frontend/ItemShop/Art/WaterBackground/T_UI_WaterFlowMap}
   FlowMap UVs
   FlowIntensity=Constant() {r=0.5}
   FlowSpeed=Constant() {r=0.01}
  ) {materialFunction=/FortUILibrary/MaterialFunctions/Effects/MF_UI_FlowMap}
  n64 = Multiply(A=Time(), B=ScalarParameter() {defaultValue=0.05, parameterName=Time Factor})
  n65 = TextureSample(UVs=Panner(Coordinate=Multiply(A=@n62.Scaled To Fill) {constB=4}, Time=@n64) {speedX=0.18, speedY=0.11}) {samplerSource=SSM_Wrap_WorldGroupSettings, texture=/Game/UI/Frontend/ItemShop/Art/IWDTakeover/Texture/T-Noise-IWD}
  n66 = NamedRerouteDeclaration(None=@n62.Scaled To Fill) {name=Fill, nodeColor=<rgba 1 0.9609 0 1>}
  n67 = TextureSample(UVs=Panner(Coordinate=Multiply(A=@n66) {constB=0.25}, Time=@n64) {speedX=0.15, speedY=0.15}) {samplerType=SAMPLERTYPE_Grayscale, texture=&LowResBlurredNoise}
  n68 = LinearInterpolate(A=@n63.Result, Alpha=Multiply(A=Add(A=@n65.R, B=@n67.R)) {constB=1.4})
  n69 = MaterialFunctionCall(Texture=@n63.Result, GradientStart=Constant() {r=0.088}, GradientEnd=Constant() {r=1.0594}, RotationAngle=Constant() {r=0.75}) {materialFunction=&MF_UI_LinearGradient}
  NamedRerouteDeclaration(None=Multiply(A=LinearInterpolate(B=@n68, Alpha=@n69.Result)) {constB=1.25}) {name=variation, nodeColor=<rgba 0 0.2969 1 1>}
  n70 = Multiply(A=NamedRerouteUsage()) {constB=0.65}
  n71 = Panner(Coordinate=@n70) {speedX=0.01, speedY=0.015}
  n72 = TextureObject() {samplerType=SAMPLERTYPE_Alpha, texture=&T_UI_CloudLV}
  n73 = TextureSample(
   UVs=Panner(Coordinate=@n70) {speedX=-0.01, speedY=-0.01}
   Tex=TextureObject() {samplerType=SAMPLERTYPE_Grayscale, texture=&T_Cloud_R}
  )
  n74 = Constant() {r=0.5}
  n75 = Constant() {r=0.09}
  n76 = Constant() {r=0.25}
  NamedRerouteDeclaration(None=ComponentMask(None=MaterialFunctionCall(UVs=@n71, Texture Object=@n72, Grayscale Flow Map=@n73.RGB, Flow Direction=@n74, Flow Intensity=@n75, Flow Speed=@n76) {materialFunction=&MF_UI_GrayscaleFlowMap_New}.Distorted Texture) {r=true}) {name=FlowMap_MG, nodeColor=<rgba 0 0 0 1>}
 ]
}
