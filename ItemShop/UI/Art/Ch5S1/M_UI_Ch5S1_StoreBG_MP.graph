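# Animated Chapter 5 Season 1 Item Shop background material, compositing clouds, lightbeams and noise variation in screen-aligned UI space.
# A radial gradient forms the light source and blends RGB_1 and RGB_2 into the background color.
# Foreground and midground clouds combine flow-mapped textures, panning noise and gradient masks, while polar-coordinate noise distortion creates drifting lightbeams.
# Named reroutes link the layers, which are color-blended, summed, desaturated and passed to MF_UI_MPItemShop_Setup with a brightness parameter.
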Material /ItemShop/UI/Art/Ch5S1/M_UI_Ch5S1_StoreBG_MP
graph {
 properties {pixelDepthOffsetMode=PDOM_Legacy, usedShadingModels=Default Lit}
 nodes [
  n1 = Power(Base=NamedRerouteUsage(), Exp=Constant() {r=0.35})
  n2 = NamedRerouteUsage()
  n3 = Power(Base=Multiply(A=NamedRerouteUsage()) {constB=2}, Exp=Constant() {r=0.775})
  n4 = NamedRerouteUsage()
  n5 = LinearInterpolate(B=Saturate(None=LinearInterpolate(B=@n2, Alpha=@n3) {constB=0}), Alpha=Saturate(None=OneMinus(None=@n4))) {constB=0}
  n6 = Constant() {r=0.65}
  n7 = Add(
   A=LinearInterpolate(A=NamedRerouteUsage(), B=VectorParameter() {parameterName=RGB_1}.RGB, Alpha=Saturate(None=@n1)) {constB=0}
   B=LinearInterpolate(
    A=NamedRerouteUsage()
    B=VectorParameter() {parameterName=RGB_2, defaultValue=<rgba 0.0727 0.1477 0.3438 1>}.RGB
    Alpha=Power(Base=@n5, Exp=@n6)
   )
  ) {constB=0.125}
  n8 = NamedRerouteUsage()
  n9 = Constant() {r=0.75}
  n10 = NamedRerouteUsage()
  n11 = Constant() {r=1}
  n12 = LinearInterpolate(B=Multiply(A=Power(Base=@n8, Exp=@n9), B=Constant() {r=8}) {constB=4}, Alpha=Saturate(None=Power(Base=@n10, Exp=@n11)))
  MaterialFunctionCall!(
   Color=Desaturation(None=Add(A=@n7, B=@n12), Fraction=Constant() {r=0.35})
   Color Brightness=ScalarParameter() {defaultValue=0.75, parameterName=FinalBrightness}
   Floor Radius
   Floor Density
   Floor Metallic
   Floor Specular
   Floor Roughness
  ) {materialFunction=/Game/UI/Foundation/Environment/Shell/Materials/MaterialFunctions/MF_UI_MPItemShop_Setup}
  n13 = Multiply(A=NamedRerouteUsage()) {constB=3}
  n14 = Multiply(A=AppendVector(A=Constant(), B=Constant() {r=0.85}), B=@n13)
  n15 = Constant() {r=0.05}
  n16 = Constant() {r=1}
  n17 = Constant() {r=2}
  n18 = MaterialFunctionCall(UVs=@n13, Texture Scale=Constant() {r=1.25}) {materialFunction=&ScaleUVsByCenter}
  n19 = Max(
   A=NamedRerouteUsage()
   B=TextureSample(UVs=Panner(Coordinate=@n18.UVs) {speedX=0.03, speedY=0.02}) {samplerType=SAMPLERTYPE_Grayscale, texture=&LowResBlurredNoise}.R
  )
  n20 = Multiply(
   A=OneMinus(None=MaterialFunctionCall!(UVs=@n14, Subtract=@n15, Falloff=@n16, Intensity=@n17) {materialFunction=/Game/UI/Frontend/ItemShop/Art/_MaterialFunctions_ShopTheming/MF_UI_TopBottomGradient}.Result)
   B=@n19
  )
  n21 = Add(A=NamedRerouteUsage(), B=AppendVector(A=Constant() {r=-0.1}, B=Constant() {r=-0.1}))
  n22 = MaterialFunctionCall(
   UVs=@n21
   GradientStart=Constant() {r=-0.1}
   GradientEnd=Constant() {r=1.25}
   RotationAngle=Constant() {r=0.8}
   GradientOffset=Constant() {r=0.6}
  ) {materialFunction=&MF_UI_LinearGradient}
  n23 = SmoothStep(Value=MaterialFunctionCall(
    UVs=@n21
    GradientStart=Constant() {r=-0.1}
    GradientEnd=Constant() {r=1.25}
    RotationAngle=Constant() {r=0.65}
    GradientOffset=Constant() {r=0.65}
   ) {materialFunction=&MF_UI_LinearGradient}.Result)
  NamedRerouteDeclaration(None=Max(A=Max(A=@n20, B=@n19), B=Max(A=@n22.Result, B=@n23))) {name=FG Clouds, nodeColor=<rgba 0.9688 0 1 1>}
  n24 = MaterialFunctionCall(Float2=MaterialFunctionCall() {materialFunction=&ScreenResolution}.Visible Resolution) {materialFunction=&BreakOutFloat2Components}
  n25 = Divide(A=@n24.R, B=@n24.R)
  n26 = Multiply(
   A=MaterialFunctionCall() {materialFunction=&ScreenAlignedUVs}."X 100%, Y 100%"
   B=AppendVector(A=@n25, B=Constant() {r=1})
  )
  n27 = ScalarParameter() {defaultValue=1, group=Radial Gradient, parameterName=Tile}
  n28 = ScalarParameter() {group=Radial Gradient, parameterName=x-Offset}
  n29 = ScalarParameter() {defaultValue=-0.75, group=Radial Gradient, parameterName=y-Offset}
  n30 = Multiply(A=@n25) {constB=0.5}
  n31 = Constant() {r=0.5}
  n32 = MaterialFunctionCall(
   UVs=NamedRerouteDeclaration(None=Multiply(A=@n26, B=@n27)) {name=SquareUV, nodeColor=<rgba 0 1 0.5078 1>}
   CenterPosition=Add(A=AppendVector(A=@n30, B=@n31), B=AppendVector(A=@n28, B=@n29))
   Radius=ScalarParameter() {defaultValue=1.5, desc="Only works if CrewLegacyUnsubscribed = 0", group=Radial Gradient, parameterName=Radius}
   Density=ScalarParameter() {defaultValue=0.65, desc="Only works if CrewLegacyUnsubscribed = 0", group=Radial Gradient, parameterName=Density}
  ) {materialFunction=&RadialGradientExponential}
  NamedRerouteDeclaration(None=@n32.RadialGradientExponential) {name=LightSourceGrad, nodeColor=<rgba 0 1 0.2734 1>}
  n33 = Multiply(A=NamedRerouteUsage()) {constB=1.25}
  n34 = Panner(Coordinate=@n33) {speedX=0.01, speedY=0.015}
  n35 = TextureObject() {samplerType=SAMPLERTYPE_Alpha, texture=/Game/UI/Frontend/ItemShop/Art/LastVoice/T_UI_CloudLV}
  n36 = TextureSample(
   UVs=Panner(Coordinate=@n33) {speedX=-0.01, speedY=-0.01}
   Tex=TextureObject() {samplerType=SAMPLERTYPE_Grayscale, texture=/FortUILibrary/MaterialFunctions/ExampleContent/T_Cloud_R}
  )
  n37 = Constant() {r=0.5}
  n38 = Constant() {r=0.09}
  n39 = Constant() {r=0.25}
  NamedRerouteDeclaration(None=ComponentMask(None=MaterialFunctionCall!(UVs=@n34, Texture Object=@n35, Grayscale Flow Map=@n36.RGB, Flow Direction=@n37, Flow Intensity=@n38, Flow Speed=@n39) {materialFunction=/FortUILibrary/MaterialFunctions/Effects/MF_UI_GrayscaleFlowMap_New}.Distorted Texture) {r=true}) {name=FlowMap_FG, nodeColor=<rgba 0 0 0 1>}
  n40 = Power(Base=NamedRerouteUsage(), Exp=Constant() {r=0.25})
  n41 = ScalarParameter() {defaultValue=1, group=Radial Gradient, parameterName=Tile}
  n42 = Multiply(A=NamedRerouteUsage(), B=@n41)
  n43 = Multiply(A=@n41) {constB=0.5}
  n44 = ScalarParameter() {defaultValue=-0.7, parameterName=offset x}
  n45 = ScalarParameter() {defaultValue=1.25, parameterName=offset y}
  n46 = MaterialFunctionCall(
   Vector or UVs=Add(A=Subtract(A=@n42, B=@n43) {constB=3}, B=AppendVector(A=@n44, B=@n45)) {constB=0.5}
   Swizzle Coordinate Output=StaticBool() {value=true}
  ) {materialFunction=&VectorToRadialValue}
  n47 = Panner(Coordinate=@n46.Vector Converted to Angle, Time=Multiply(A=Time(), B=ScalarParameter() {defaultValue=0.05, parameterName=TimeFactor2})) {speedX=0.111, speedY=0.1}
  n48 = Multiply(
   A=TextureSample(UVs=Multiply(A=@n47) {constB=2}) {samplerType=SAMPLERTYPE_Grayscale, texture=/Game/UI/Frontend/ItemShop/Art/Rumble/LowResBlurredNoise_Rumble}.R
   B=ScalarParameter() {defaultValue=0.0125, parameterName=DeformIntensity}
  )
  n49 = Multiply(
   A=Power(
    Base=TextureSample(UVs=@n48) {samplerType=SAMPLERTYPE_Grayscale, texture=&LowResBlurredNoise_Rumble}.R
    Exp=Constant() {r=3}
   ) {constExponent=1}
   B=Constant() {r=1.5}
  ) {constB=2}
  n50 = Panner(Coordinate=@n46.Vector Converted to Angle, Time=Multiply(A=Time(), B=ScalarParameter() {defaultValue=0.025, parameterName=TimeFactor})) {speedX=0.025, speedY=-0.025}
  n51 = Multiply(
   A=TextureSample(UVs=Multiply(A=@n50) {constB=2}) {samplerType=SAMPLERTYPE_Grayscale, texture=&LowResBlurredNoise_Rumble}.R
   B=ScalarParameter() {defaultValue=0.25, parameterName=DeformIntensity_01}
  )
  n52 = Multiply(
   A=Power(
    Base=TextureSample(UVs=@n51) {samplerType=SAMPLERTYPE_Grayscale, texture=&LowResBlurredNoise_Rumble}.R
    Exp=Constant() {r=3}
   ) {constExponent=1}
   B=Constant() {r=1.5}
  ) {constB=2}
  n53 = Saturate(None=Min(A=Saturate(None=@n49), B=Saturate(None=@n52)))
  NamedRerouteDeclaration(None=Multiply(A=VectorParameter() {parameterName=RGB_2, defaultValue=<rgba 0.0727 0.1477 0.3438 1>}.RGB, B=Multiply(A=@n40, B=@n53))) {name=Lightbeams, nodeColor=<rgba 1 0.8203 0 1>}
  n54 = Multiply(A=NamedRerouteUsage()) {constB=3}
  n55 = Multiply(A=AppendVector(A=Constant(), B=Constant() {r=0.85}), B=@n54)
  n56 = Constant() {r=0.05}
  n57 = Constant() {r=1}
  n58 = Constant() {r=2}
  n59 = MaterialFunctionCall(UVs=@n54, Texture Scale=Constant() {r=1.5}) {materialFunction=&ScaleUVsByCenter}
  n60 = Max(
   A=NamedRerouteUsage()
   B=TextureSample(UVs=Panner(Coordinate=@n59.UVs) {speedX=0.015}) {samplerType=SAMPLERTYPE_Grayscale, texture=&LowResBlurredNoise}.R
  )
  n61 = Multiply(
   A=OneMinus(None=MaterialFunctionCall(UVs=@n55, Subtract=@n56, Falloff=@n57, Intensity=@n58) {materialFunction=&MF_UI_TopBottomGradient}.Result)
   B=@n60
  )
  n62 = Constant() {r=-0.1}
  n63 = Constant() {r=1}
  n64 = Constant() {r=0.8}
  n65 = Constant() {r=0.7}
  n66 = Add(A=NamedRerouteUsage(), B=AppendVector(A=Constant(), B=Constant() {r=0.6}))
  n67 = Constant() {r=-0.1}
  n68 = Constant() {r=0.75}
  n69 = Constant() {r=0.65}
  n70 = Constant() {r=0.65}
  n71 = SmoothStep(Value=Max(
    A=MaterialFunctionCall(UVs=@n66, GradientStart=@n62, GradientEnd=@n63, RotationAngle=@n64, GradientOffset=@n65) {materialFunction=&MF_UI_LinearGradient}.Result
    B=MaterialFunctionCall(UVs=@n66, GradientStart=@n67, GradientEnd=@n68, RotationAngle=@n69, GradientOffset=@n70) {materialFunction=&MF_UI_LinearGradient}.Result
   )) {constMax=3}
  n72 = Multiply(A=Max(A=Max(A=@n61, B=@n60), B=@n71), B=@n71)
  NamedRerouteDeclaration(None=Multiply(A=Saturate(None=@n72)) {constB=4}) {name=MG_Clouds, nodeColor=<rgba 0.9688 0 1 1>}
  n73 = NamedRerouteUsage()
  n74 = MaterialFunctionCall!(
   Texture 2D=TextureObject() {samplerType=SAMPLERTYPE_Grayscale, texture=&LowResBlurredNoise_Rumble}
   Texture UVs=Panner(Coordinate=Multiply(A=@n73) {constB=1.2}) {speedX=-0.02, speedY=-0.013}
   FlowMap Texture 2D=TextureObject() {samplerType=SAMPLERTYPE_LinearColor, texture=/Game/UI/Frontend/ItemShop/Art/WaterBackground/T_UI_WaterFlowMap}
   FlowMap UVs
   FlowIntensity=Constant() {r=0.5}
   FlowSpeed=Constant() {r=0.01}
  ) {materialFunction=/FortUILibrary/MaterialFunctions/Effects/MF_UI_FlowMap}
  n75 = Multiply(A=Time(), B=ScalarParameter() {defaultValue=0.125, parameterName=Time Factor})
  n76 = TextureSample(UVs=Panner(Coordinate=Multiply(A=@n73) {constB=4}, Time=@n75) {speedX=0.18, speedY=0.11}) {samplerSource=SSM_Wrap_WorldGroupSettings, texture=/Game/UI/Frontend/ItemShop/Art/IWDTakeover/Texture/T-Noise-IWD}
  n77 = TextureSample(UVs=Panner(Coordinate=Multiply(A=@n73) {constB=0.25}, Time=@n75) {speedX=0.15, speedY=0.15}) {samplerType=SAMPLERTYPE_Grayscale, texture=&LowResBlurredNoise}
  n78 = LinearInterpolate(A=@n74.Result, Alpha=Multiply(A=Add(A=@n76.R, B=@n77.R)) {constB=1.4})
  n79 = MaterialFunctionCall(
   UVs=NamedRerouteUsage()
   Texture=@n74.Result
   GradientStart=Constant() {r=0.088}
   GradientEnd=Constant() {r=1.0594}
   RotationAngle=Constant() {r=0.75}
  ) {materialFunction=&MF_UI_LinearGradient}
  NamedRerouteDeclaration(None=Multiply(A=LinearInterpolate(B=@n78, Alpha=@n79.Result)) {constB=1.25}) {name=variation, nodeColor=<rgba 0 0.2969 1 1>}
  NamedRerouteDeclaration(None=LinearInterpolate(
    A=VectorParameter() {parameterName=RGB_1}.RGB
    B=VectorParameter() {parameterName=RGB_2, defaultValue=<rgba 0.0727 0.1477 0.3438 1>}.RGB
    Alpha=@n32.RadialGradientExponential
   )) {name=BackgroundColor, nodeColor=<rgba 0 0 0 1>}
  n80 = Multiply(A=NamedRerouteUsage()) {constB=0.65}
  n81 = Panner(Coordinate=@n80) {speedX=0.01, speedY=0.015}
  n82 = TextureObject() {samplerType=SAMPLERTYPE_Alpha, texture=&T_UI_CloudLV}
  n83 = TextureSample(
   UVs=Panner(Coordinate=@n80) {speedX=-0.01, speedY=-0.01}
   Tex=TextureObject() {samplerType=SAMPLERTYPE_Grayscale, texture=&T_Cloud_R}
  )
  n84 = Constant() {r=0.5}
  n85 = Constant() {r=0.09}
  n86 = Constant() {r=0.25}
  NamedRerouteDeclaration(None=ComponentMask(None=MaterialFunctionCall(UVs=@n81, Texture Object=@n82, Grayscale Flow Map=@n83.RGB, Flow Direction=@n84, Flow Intensity=@n85, Flow Speed=@n86) {materialFunction=&MF_UI_GrayscaleFlowMap_New}.Distorted Texture) {r=true}) {name=FlowMap_MG, nodeColor=<rgba 0 0 0 1>}
 ]
}
